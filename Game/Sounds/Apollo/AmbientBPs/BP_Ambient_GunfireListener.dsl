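; Ambient listener actor that answers nearby pawns' weapon fire with a sound placed away from the local player.
; Sphere overlaps, handled only on CanExecuteCosmeticEvents' False branch, bind or unbind weapon-fire and equip delegates, following weapon swaps.
; Shots passing a chance roll and cooldown play the sound after a random delay, randomly offset horizontally from the cached local pawn.
; Construction script sets sphere radius; debug overlap prints remain.
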(event Custom|PawnOverlapBegin
  (params (Pawn :type Actor%20Object%20Reference))
  (bind _asfort_pawn (Utilities|Casting|CastToFortPawn Pawn))
  (Utilities|FlowControl|Sequence
    (:then_0
      (if (Pawn|IsLocallyControlled _asfort_pawn)
        (Variables|Default|SetLocalPawn _asfort_pawn)))
    (:then_1
      (FortPawn|BindEventtoOnWeaponEquippedDelegate _asfort_pawn (EventDispatchers|CreateEvent :!event Custom|On%20Pawn%20Weapon%20Equipped))
      (Utilities|IsValid (Variables|/Script/FortniteGame.FortPawn|GetCurrentWeapon _asfort_pawn)
        (:Is%20Valid
          (EventDispatchers|BindEventtoOnWeaponFireAudioVisuals (Variables|/Script/FortniteGame.FortPawn|GetCurrentWeapon _asfort_pawn) (EventDispatchers|CreateEvent :!event Custom|On%20Weapon%20Fire))
          (Development|PrintString "Overlap in" true true "(R=1.000000,G=0.000000,B=0.042602,A=1.000000)"))))))

(event Custom|PawnOverlapEnd
  (params (Pawn :type Actor%20Object%20Reference))
  (bind _asfort_pawn (Utilities|Casting|CastToFortPawn Pawn))
  (Utilities|FlowControl|Sequence
    (:then_0
      (FortPawn|UnbindEventfromOnWeaponEquippedDelegate _asfort_pawn (EventDispatchers|CreateEvent :!event Custom|On%20Pawn%20Weapon%20Equipped))
      (Utilities|IsValid (Variables|/Script/FortniteGame.FortPawn|GetCurrentWeapon _asfort_pawn)
        (:Is%20Valid
          (EventDispatchers|UnbindEventfromOnWeaponFireAudioVisuals (Variables|/Script/FortniteGame.FortPawn|GetCurrentWeapon _asfort_pawn) (EventDispatchers|CreateEvent :!event Custom|On%20Weapon%20Fire))
          (Development|PrintString "Overlap out" true true "(R=1.000000,G=0.000000,B=0.042602,A=1.000000)"))))
    (:then_1
      (if (Pawn|IsLocallyControlled _asfort_pawn)
        (Variables|Default|SetLocalPawn 0)))))

(event Custom|On%20Weapon%20Fire
  (params (Weapon :type Fort%20Weapon%20Object%20Reference))
  (bind (_randomdelayrange_x _randomdelayrange_y) (Math|Vector2D|BreakVector2D (Variables|GetRandomDelayRange_X)))
  (if (and (< (Math|Random|RandomFloatinRange 0.0 1.0) (Variables|Default|GetPercentChance)) (>= (- (Utilities|Time|GetGameTimeinSeconds) (Variables|Default|GetLastAudioPlayedTime)) (Variables|Default|GetCooldownTimeThreshold)))
    (Variables|Default|SetLastAudioPlayedTime (Utilities|Time|GetGameTimeinSeconds))
    (Utilities|FlowControl|Delay (Math|Random|RandomFloatinRange _randomdelayrange_x _randomdelayrange_y))
    (bind _localpawn (Variables|Default|GetLocalPawn)
      (:then
        (Audio|PlaySoundatLocation (Variables|Default|GetSoundToPlay) (+ (Transformation|GetActorLocation _localpawn) (* (Math|Vector|MakeVector (Math|Random|RandomFloatinRange -1.0 1.0) (Math|Random|RandomFloatinRange -1.0 1.0)) (Math|Random|RandomFloatinRange 2e+03 4e+03)))))
      (:else))))

(event Custom|On%20Pawn%20Weapon%20Equipped
  (params (NewWeapon :type Fort%20Weapon%20Object%20Reference) (PrevWeapon :type Fort%20Weapon%20Object%20Reference))
  (bind _outputdelegate (EventDispatchers|CreateEvent :!event Custom|On%20Weapon%20Fire))
  (EventDispatchers|BindEventtoOnWeaponFireAudioVisuals NewWeapon _outputdelegate)
  (Utilities|IsValid PrevWeapon
    (:Is%20Valid
      (EventDispatchers|UnbindEventfromOnWeaponFireAudioVisuals PrevWeapon _outputdelegate))))

(event OnComponentBeginOverlap(OverlapSphere)
  (params (OverlappedComponent :type Primitive%20Component%20Object%20Reference) (OtherActor :type Actor%20Object%20Reference) (OtherComp :type Primitive%20Component%20Object%20Reference) (OtherBodyIndex :type Integer) (bFromSweep :type Boolean) (SweepResult :type Hit%20Result%20Structure%20%28by%20ref%29))
  (Networking|CanExecuteCosmeticEvents
    (:False
      (CallFunction|PawnOverlapBegin OtherActor))))

(event OnComponentEndOverlap(OverlapSphere)
  (params (OverlappedComponent :type Primitive%20Component%20Object%20Reference) (OtherActor :type Actor%20Object%20Reference) (OtherComp :type Primitive%20Component%20Object%20Reference) (OtherBodyIndex :type Integer))
  (Networking|CanExecuteCosmeticEvents
    (:False
      (CallFunction|PawnOverlapEnd OtherActor))))

(fn ConstructionScript
  (Components|Sphere|SetSphereRadius (Variables|Default|GetOverlapSphere) (Variables|Default|GetRadius) false))
