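; Default ambient audio Blueprint that, on non-authority instances, registers an ambient entry and adapts ambience to fog weather.
; Once the time-of-day manager is ready, it binds to weather changes, deferring through a registration event if the weather component is unset.
; Blend-in sets a light, medium or heavy fog tag by event index, pushes a fog sound mix and crossfades weather audio in.
; Blend-out reverses mix and audio; inactive clears tags.
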(event EventBeginPlay
  ;[0.107 0.1559 0.0999] Begin Play
  (Networking|SwitchHasAuthority
    (:Remote
      (AmbientAudio|AddAmbientEntry (WorldSubsystems|GetAmbientAudioSubsystem) (Variables|Default|GetAmbientName) (Variables|Default|GetDefaultAudioBank) (Variables|Default|GetDefaultAudioBankPriority) 0.0)
      (bind _timeofdaymanager (Class|FortAsyncActionTimeOfDayManagerReady|GetTimeOfDayManagerReady)
        (:OnReady
          (if (Class|ItemUIFunctionLibrary|IsNullItem (TimeOfDay|GetWeatherComponent _timeofdaymanager))
            (TimeOfDay|RegisterforWeatherComponentSetEvent _timeofdaymanager (EventDispatchers|CreateEvent :!event Custom|WeatherComponentReady))
            (else
              (Events|BindEventtoOnWeatherChanged (TimeOfDay|GetWeatherComponent _timeofdaymanager) (EventDispatchers|CreateEvent :!event Custom|WeatherChanged)))))))))

(event Custom|WeatherComponentReady
  (params (WeatherComponent :type Fort%20Time%20Of%20Day%20Weather%20Component%20Object%20Reference))
  ;[0.0865 0.009134 0.008568] Bind To Weather Changed
  (Events|BindEventtoOnWeatherChanged WeatherComponent (EventDispatchers|CreateEvent :!event Custom|WeatherChanged)))

(event Custom|WeatherChanged
  (params (WeatherState :type EGlobalWeatherState%20Enum :default "Inactive") (WeatherData :type Fort%20Replicated%20Weather%20Event%20Data%20Structure))
  (bind (_index _starttime _endtime _blendinduration _blendoutduration) (Utilities|Struct|BreakFortReplicatedWeatherEventData WeatherData))
  (bind _ambientaudio_weather (Variables|Default|GetAmbientAudio_Weather))
  (bind _ambientaudio_weather_2 (Variables|Default|GetAmbientAudio_Weather))
  ;[0.0865 0.009134 0.008568] Handle Weather Changed
  (switch Utilities|FlowControl|Switch|SwitchonEGlobalWeatherState WeatherState
    (:Inactive
      (Development|PrintString "Inactive")
      (CallFunction|ClearFogTags))
    (:Active
      (Development|PrintString "Active"))
    (:BlendingIn
      (Development|PrintString "BlendIn")
      (CallFunction|SetWeatherTags _index)
      (Audio|PushSoundMixModifier "/Game/Sounds/Apollo/Biomes/Fog/Fog_Amb_SoundMix.Fog_Amb_SoundMix")
      (Variables|/Script/AmbientAudio.AmbientAudioComponent|SetCrossfadeTime _blendinduration _ambientaudio_weather_2)
      (Components|Activation|Activate _ambientaudio_weather_2 false))
    (:BlendingOut
      (Development|PrintString "BlendOut")
      (Audio|PopSoundMixModifier "/Game/Sounds/Apollo/Biomes/Fog/Fog_Amb_SoundMix.Fog_Amb_SoundMix")
      (Variables|/Script/AmbientAudio.AmbientAudioComponent|SetCrossfadeTime _blendoutduration _ambientaudio_weather)
      (Components|Activation|Deactivate _ambientaudio_weather))))

(fn ClearFogTags
  (bind _returnvalue (WorldSubsystems|GetAmbientAudioSubsystem))
  (AmbientAudio|RemoveGameplayTag _returnvalue "(TagName=\"AmbientAudio.Weather.Fog.Light\")")
  (AmbientAudio|RemoveGameplayTag _returnvalue "(TagName=\"AmbientAudio.Weather.Fog.Medium\")")
  (AmbientAudio|RemoveGameplayTag _returnvalue "(TagName=\"AmbientAudio.Weather.Fog.Heavy\")")
  (return))

(fn SetWeatherTags
  (params (Selection :type Integer))
  (bind _returnvalue (WorldSubsystems|GetAmbientAudioSubsystem))
  (CallFunction|ClearFogTags)
  (switch int Selection
    (:Default
      _)
    (:0
      (AmbientAudio|AddGameplayTag _returnvalue "(TagName=\"AmbientAudio.Weather.Fog.Light\")"))
    (:1
      (AmbientAudio|AddGameplayTag _returnvalue "(TagName=\"AmbientAudio.Weather.Fog.Medium\")"))
    (:2
      (AmbientAudio|AddGameplayTag _returnvalue "(TagName=\"AmbientAudio.Weather.Fog.Heavy\")")))
  (return))

(fn ConstructionScript)
