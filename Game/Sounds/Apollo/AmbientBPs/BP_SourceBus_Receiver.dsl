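; Ambient audio Blueprint that plays a source bus through an audio component while attached to a building actor.
; On BeginPlay it assigns the source bus asset variable as the component's sound, starts playback, and casts its attach parent to BuildingSMActor.
; It binds to that building's destroyed and died events.
; Either event, or its own EndPlay, triggers a short linear fade-out of the audio.
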(event EventBeginPlay
  (bind _sourcebus (Variables|Default|GetSourceBus))
  (Audio|Components|Audio|SetSound _sourcebus (Variables|Default|GetSourceBusAsset))
  (Audio|Components|Audio|Play _sourcebus 0.0)
  (bind _asbuilding_static_mesh (Utilities|Casting|CastToBuildingSMActor (Actor|GetAttachParentActor)))
  (Game|BindEventtoOnDestroyed _asbuilding_static_mesh (delegate Custom|On%20Owner%20Destroyed))
  (EventDispatchers|BindEventtoOnDied _asbuilding_static_mesh (delegate Custom|On%20Owner%20Died)))

(event EventEndPlay
  (params (EndPlayReason :type EEndPlayReason%20Enum :default "Destroyed"))
  (Audio|Components|Audio|FadeOut (Variables|Default|GetSourceBus) 0.2 0.0 "Linear" :!class /Script/Engine.AudioComponent))

(event Custom|On%20Owner%20Destroyed
  (params (DestroyedActor :type Actor%20Object%20Reference))
  (Audio|Components|Audio|FadeOut (Variables|Default|GetSourceBus) 0.2 0.0 "Linear" :!class /Script/Engine.AudioComponent))

(event Custom|On%20Owner%20Died
  (params (DamagedActor :type Actor%20Object%20Reference) (Damage :type Float%20%28single-precision%29) (InstigatedBy :type Controller%20Object%20Reference) (DamageCauser :type Actor%20Object%20Reference) (HitLocation :type Vector) (FHitComponent :type Primitive%20Component%20Object%20Reference) (BoneName :type Name) (Momentum :type Vector))
  (Audio|Components|Audio|FadeOut (Variables|Default|GetSourceBus) 0.2 0.0 "Linear" :!class /Script/Engine.AudioComponent))

(fn ConstructionScript)
