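; Parent Blueprint for ambient audio that switches on or off according to the ocean flood level.
; On BeginPlay it waits for playlist data, then, where cosmetic events can run, converts the water subsystem's flood height into a level index in 384-unit steps.
; It prints that index and calls PlayAudio if ActivatedWaterLevels contains it, otherwise StopAudio.
; Both functions are empty here, and the level is not checked again after startup.
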(event EventBeginPlay
  (bind _returnvalue (WorldSubsystems|GetWaterSubsystem))
  (bind _returnvalue_2 (Water|GetOceanFloodHeight _returnvalue))
  (bind _returnvalue_3 (Math|Float|Truncate _returnvalue_2))
  (bind _returnvalue_4 (/ _returnvalue_3 384))
  (Class|FortAsyncActionCurrentPlaylistReady|PlaylistDataReadyAsync
    (:OnReady
      (Networking|CanExecuteCosmeticEvents
        (:True
          (Development|PrintString (Utilities|String|ToString(Integer) _returnvalue_4))
          (if (Utilities|Array|ContainsItem (Variables|Default|GetActivatedWaterLevels) _returnvalue_4)
            (CallFunction|PlayAudio)
            (else
              (CallFunction|StopAudio))))))))

(fn PlayAudio)

(fn StopAudio)

(fn ConstructionScript)
